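; Fortnite camera modifier that supplies a look-at target rotation for the view target, based on either pawn velocity or actor facing.
; The target rotation points along movement velocity when UsePawnVelocity is set, otherwise along actor forward tilted by DefaultCameraPitch.
; The same flag swaps the parent's relative pitch frustum for VelocityBasedCameraFrustum.
; The custom normal uses only control-rotation yaw, tilted by the default pitch.
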(fn GetRelativePitchFrustum
  (bind _outrelativepitchfrustum (|CallParentFunction))
  (return (select (Variables|Default|GetUsePawnVelocity) (Variables|Default|GetVelocityBasedCameraFrustum) _outrelativepitchfrustum)))

(fn GetCustomNormal
  (bind (_returnvalue_roll _returnvalue_pitch _returnvalue_yaw) (Math|Rotator|BreakRotator (Pawn|GetControlRotation (Utilities|Casting|CastToFortPawn (CameraModifier|GetViewTarget)) :!class /Script/Engine.Pawn)))
  (return (Math|Vector|RotateVectorAroundAxis (Math|Vector|GetForwardVector (Math|Rotator|MakeRotator 0.0 0.0 _returnvalue_yaw)) (Variables|Default|GetDefaultCameraPitch) (Math|Vector|GetRightVector (Math|Rotator|MakeRotator 0.0 0.0 _returnvalue_yaw)))))

(fn GetTargetCameraRotation
  (bind _returnvalue (CameraModifier|GetViewTarget))
  (bind _asfort_player_pawn_athena (Utilities|Casting|CastToFortPlayerPawnAthena _returnvalue))
  (return (Math|Rotator|MakeRotfromX (Math|Vector|Normalize (Math|Vector|SelectVector (Variables|/Script/Engine.MovementComponent|GetVelocity (Variables|/Script/Engine.Character|GetCharacterMovement _asfort_player_pawn_athena)) (Math|Vector|RotateVectorAroundAxis (Transformation|GetActorForwardVector _returnvalue) (Variables|Default|GetDefaultCameraPitch) (Transformation|GetActorRightVector _returnvalue)) (Variables|Default|GetUsePawnVelocity))))))
