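; Gameplay ability that activates ScubaBond Oak for non-leader users through a temporary non-leader weapon.
; On activation it saves the current weapon, equips that weapon definition without animation, then waits on a client-wait net sync before manually activating another ability class.
; It ends on the Oak completed event, sharing that ability's lifetime.
; On end it removes the fake weapon and re-equips the previous or best weapon on non-authority clients only, avoiding server-cached fake-weapon state.
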(event Ability|EventActivateAbility
  (Variables|Default|SetPreviousWeapon (Weapon|GetCurrentWeapon (Ability|GetActivatingPawn)))
  (bind _returnvalue (Weapon|EquipWeaponDefinition :self (CallFunction|GetAvatarasFortPlayerPawn) :WeaponData "/ScubaBondOakGameplay/Gameplay/WID_ScubaBond_Oak_NonLeader.WID_ScubaBond_Oak_NonLeader" :bDisableEquipAnimation true)) ; Equipping our non-leader version of the WID
  (Variables|Default|SetFakeWeapon _returnvalue)
  (Ability|Tasks|WaitNetSync "OnlyClientWait"
    (:OnSync
      (Abilities|TryActivateAbilitybyClass (Ability|GetAbilitySystemComponentfromActorInfo) (Variables|Default|GetAtivation%20Class) true) ; Manually Calling Activation for non-leader users
      ;[1 1 1] Using the same end logic as ScubaBondOak_Activate because we want the same lifetime
      (Ability|Tasks|WaitGameplayEvent "(TagName=\"Event.ScubaBond.Oak.Ability.Completed\")" 0 true
        (:EventReceived
          (Ability|EndAbility))))) ; Waiting Until We're equipped to run the function so that \nHolsters do not run into race conditions
)

(event Ability|EventOnEndAbility
  (params (bWasCancelled :type Boolean))
  (bind _returnvalue (CallFunction|GetAvatarasFortPlayerPawn))
  (bind _fakeweapon (Variables|Default|GetFakeWeapon))
  (Utilities|IsValid _fakeweapon
    (:Is%20Valid
      (Inventory|RemoveFortItemFromPlayer (Utilities|Casting|CastToFortPlayerController (Ability|GetPlayerControllerfromActorInfo)) (FortWeapon|GetInventoryItem _fakeweapon) -1 true))
    (:Is%20Not%20Valid
      _))
  (Ability|HasAuthority
    (:False
      (bind _returnvalue_2 (Weapon|EquipWeaponActor _returnvalue (Variables|Default|GetPreviousWeapon))) ; Equip the last weapon that we had equipped \ndo we don't end up empty handed
      (if (not _returnvalue_2)
        (Class|FortPawn|EquipBestWeapon _returnvalue)))) ; Given that this runs on client first, \nwe can only re-equip on clients \nbecause server will cache the \nfake weapon as previous weapon
)
